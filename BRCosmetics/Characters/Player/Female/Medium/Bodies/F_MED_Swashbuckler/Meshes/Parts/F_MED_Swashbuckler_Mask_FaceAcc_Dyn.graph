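# Physics asset for the female medium Swashbuckler mask face accessory, defining bodies and joints for its dynamic hair bones.
# Six hair chains at the back, sides and front right hang from the head as two or three linked box segments, most ending in capsules.
# Hair joints allow wide limited swing but very little twist.
# Simple head, neck, clavicle and upper-spine shapes form the body, joined by free-rotation constraints.
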PhysicsAsset /BRCosmetics/Characters/Player/Female/Medium/Bodies/F_MED_Swashbuckler/Meshes/Parts/F_MED_Swashbuckler_Mask_FaceAcc_Dyn
bodies {
 clavicle_l [{shapeType=Capsule, length=7.2935, radius=7.3987, center=<5.5581 -0.3501 -2.0681>, rotation=<rot 73.3944 0 0>}]
 clavicle_r [{shapeType=Capsule, length=5.4163, radius=8.3525, center=<-5.0312 0.4592 2.508>, rotation=<rot 78.3944 0 0>}]
 dyn_hair_bk_01 [{shapeType=Box, extentX=3.7795, extentY=2.8226, extentZ=7.5124, center=<2.3436 0.0001 -0.0001>}]
 dyn_hair_bk_02 [{shapeType=Box, extentX=6.6386, extentY=3.3619, extentZ=10.7522, center=<3.2994 0 -0.0001>}]
 dyn_hair_bk_03 [{shapeType=Box, extentX=12.4028, extentY=4.6812, extentZ=10.7522, center=<5.8156 1.6638 0.0001>, rotation=<rot 0 15.0007 0>}]
 dyn_hair_bk_l_01 [{shapeType=Box, extentX=5.9004, extentY=1.2856, extentZ=7.5124, center=<3.2348 0 -0.0001>}]
 dyn_hair_bk_l_02 [{shapeType=Box, extentX=5.9004, extentY=1.2856, extentZ=7.5124, center=<3.2348 0 -0.0001>}]
 dyn_hair_bk_l_03 [{shapeType=Capsule, length=2.8876, radius=2.1124, center=<2.248 -0.4018 -0.0004>}]
 dyn_hair_bk_r_01 [{shapeType=Box, extentX=3.9387, extentY=1.8791, extentZ=7.5124, center=<2.0231 0.0001 -0.0001>}]
 dyn_hair_bk_r_02 [{shapeType=Box, extentX=5.2227, extentY=4.4433, extentZ=7.5124, center=<2.652 0.0001 -0.0001>}]
 dyn_hair_bk_r_03 [{shapeType=Capsule, length=6.8699, radius=2.8858, center=<2.8855 -0.0001 1.6373>, rotation=<rot -34.9998 0 0>}]
 dyn_hair_fr_r_01 [{shapeType=Box, extentX=5.9004, extentY=1.8791, extentZ=7.5124, center=<3.2348 0 -0.0001>}]
 dyn_hair_fr_r_02 [{shapeType=Box, extentX=5.9004, extentY=1.8791, extentZ=7.5124, center=<3.2348 0 -0.0001>}]
 dyn_hair_l_01 [{shapeType=Box, extentX=5.9004, extentY=1.6203, extentZ=7.5124, center=<3.2348 0 -0.0001>}]
 dyn_hair_l_02 [{shapeType=Box, extentX=5.2096, extentY=1.8791, extentZ=7.5124, center=<3.2348 0 -0.0001>}]
 dyn_hair_l_03 [{shapeType=Capsule, length=4.1045, radius=1.6328, center=<3.2556 1.9047 -0.0002>, rotation=<rot 90 -165.9637 139.0377>}]
 dyn_hair_r_01 [{shapeType=Box, extentX=5.9004, extentY=1.8791, extentZ=7.5124, center=<3.2348 0 -0.0001>}]
 dyn_hair_r_02 [{shapeType=Box, extentX=3.5171, extentY=1.8791, extentZ=7.5124, center=<1.8404 0 -0.0001>}]
 dyn_hair_r_03 [{shapeType=Capsule, length=7.9311, radius=1.7424, center=<1.847 -0.977 -0.0003>}]
 head [
  {shapeType=Sphere, radius=6.4177, center=<1.3282 -0.1209 -0.2823>}
  {shapeType=Capsule, length=5.9657, radius=5.522, center=<0.8118 1.1771 0>, rotation=<rot 90 0 0>}
 ]
 neck_01 [{shapeType=Capsule, length=8.2568, radius=5.3864, center=<0 0.5387 0>, rotation=<rot 90 0 0>}]
 spine_05 [{shapeType=Capsule, length=0.9494, radius=8.7688, center=<6.8632 2.0625 0>}]
}
constraints [
 {bone1Name=dyn_hair_bk_l_01, bone2Name=head, swing1LimitDegrees=60, swing2LimitDegrees=60, twistLimitDegrees=5}
 {bone1Name=dyn_hair_bk_l_02, bone2Name=dyn_hair_bk_l_01, swing1LimitDegrees=60, swing2LimitDegrees=60, twistLimitDegrees=5}
 {bone1Name=dyn_hair_bk_l_03, bone2Name=dyn_hair_bk_l_02, swing1LimitDegrees=60, swing2LimitDegrees=60, twistLimitDegrees=5}
 {bone1Name=dyn_hair_bk_01, bone2Name=head, swing1LimitDegrees=60, swing2LimitDegrees=60, twistLimitDegrees=5}
 {bone1Name=dyn_hair_bk_02, bone2Name=dyn_hair_bk_01, swing1LimitDegrees=60, swing2LimitDegrees=60, twistLimitDegrees=5}
 {bone1Name=dyn_hair_bk_03, bone2Name=dyn_hair_bk_02, swing1LimitDegrees=60, swing2LimitDegrees=60, twistLimitDegrees=5}
 {bone1Name=dyn_hair_l_01, bone2Name=head, swing1LimitDegrees=60, swing2LimitDegrees=45, twistLimitDegrees=5}
 {bone1Name=dyn_hair_l_02, bone2Name=dyn_hair_l_01, swing1LimitDegrees=60, swing2LimitDegrees=45, twistLimitDegrees=5}
 {bone1Name=dyn_hair_l_03, bone2Name=dyn_hair_l_02, swing1LimitDegrees=60, swing2LimitDegrees=45, twistLimitDegrees=5}
 {bone1Name=dyn_hair_bk_r_01, bone2Name=head, swing1LimitDegrees=60, swing2LimitDegrees=60, twistLimitDegrees=5}
 {bone1Name=dyn_hair_bk_r_02, bone2Name=dyn_hair_bk_r_01, swing1LimitDegrees=60, swing2LimitDegrees=60, twistLimitDegrees=5}
 {bone1Name=dyn_hair_bk_r_03, bone2Name=dyn_hair_bk_r_02, swing1LimitDegrees=60, swing2LimitDegrees=60, twistLimitDegrees=5}
 {bone1Name=dyn_hair_r_01, bone2Name=head, swing1LimitDegrees=60, swing2LimitDegrees=60, twistLimitDegrees=5}
 {bone1Name=dyn_hair_r_02, bone2Name=dyn_hair_r_01, swing1LimitDegrees=60, swing2LimitDegrees=60, twistLimitDegrees=5}
 {bone1Name=dyn_hair_r_03, bone2Name=dyn_hair_r_02, swing1LimitDegrees=60, swing2LimitDegrees=60, twistLimitDegrees=5}
 {bone1Name=dyn_hair_fr_r_01, bone2Name=head, swing1LimitDegrees=60, swing2LimitDegrees=45, twistLimitDegrees=5}
 {bone1Name=dyn_hair_fr_r_02, bone2Name=dyn_hair_fr_r_01, swing1LimitDegrees=60, swing2LimitDegrees=45, twistLimitDegrees=5}
 {
  bone1Name=clavicle_l
  bone2Name=spine_05
  swing1LimitDegrees=45
  swing1Motion=Free
  swing2LimitDegrees=45
  swing2Motion=Free
  twistLimitDegrees=45
  twistMotion=Free
 }
 {
  bone1Name=clavicle_r
  bone2Name=spine_05
  swing1LimitDegrees=45
  swing1Motion=Free
  swing2LimitDegrees=45
  swing2Motion=Free
  twistLimitDegrees=45
  twistMotion=Free
 }
 {
  bone1Name=head
  bone2Name=neck_01
  swing1LimitDegrees=45
  swing1Motion=Free
  swing2LimitDegrees=45
  swing2Motion=Free
  twistLimitDegrees=45
  twistMotion=Free
 }
 {
  bone1Name=neck_01
  bone2Name=spine_05
  swing1LimitDegrees=45
  swing1Motion=Free
  swing2LimitDegrees=45
  swing2Motion=Free
  twistLimitDegrees=45
  twistMotion=Free
 }
]
