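# Animation Blueprint for the female Swashbuckler's mask face accessory, adding secondary physics motion on top of the character's pose.
# It copies the attached parent mesh's pose, then runs a RigidBody node with the accessory's physics asset, simulated relative to the head bone with overridden gravity and skydiving external force.
# Each update lowers the component acceleration and velocity scale while targeting, reducing how much character movement drives the simulation.
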AnimBlueprint /BRCosmetics/Characters/Player/Female/Medium/Bodies/F_MED_Swashbuckler/Meshes/Parts/F_MED_Swashbuckler_Mask_FaceAcc_AnimBP
target_skeleton=&/F_MED_Swashbuckler_Mask_FaceAcc_Skeleton
graphs [
 AnimGraph {nodes [
   n1 = Get acc_scaler_vector() -> (acc_scaler_vector:Vector)
   n2 = RigidBody!(
    ComponentPose=Local To Component(LocalPose=Copy Pose From Mesh() {bUseAttachedParent=true})
    Alpha
    ExternalForce=Break Fort Anim Input Skydiving External Force(FortAnimInput_SkydivingExternalForce=Get SkydivingExternalForceInput()) -> (
     bUseSkydivingVectorForce=~
     SkydivingVectorForce=~
     HeadToPelvisDirection=~
     FloatingMultiplier=~
     FloatingAdditive=~
     DivingMultiplier=~
     DivingAdditive=~
     ParachutingMultiplier=~
     ParachutingAdditive=~
    ).SkydivingVectorForce
    ComponentLinearAccScale:Vector=@n1
    ComponentLinearVelScale:Vector=@n1
   ) -> (Pose) {
    bOverrideWorldGravity=true
    overridePhysicsAsset=&/F_MED_Swashbuckler_Mask_FaceAcc_Dyn
    simulationSpace=BaseBoneSpace
    alphaBoolBlend {blendOption=Linear}
    baseBoneRef {boneName=head}
    componentAppliedLinearAccClamp=<2000 2000 2000>
    componentLinearAccScale=<0.2 0.5 0.2>
    componentLinearVelScale=<0.2 0.5 0.2>
    overrideWorldGravity {z=-980}
   }
   Output Pose(Result=Component To Local(ComponentPose=@n2)) {layerGroup=None}
  ]}
 EventGraph {nodes [
   Event BlueprintUpdateAnimation()
   > Set acc_scaler_vector(acc_scaler_vector:Vector=SelectVector(A="0.1,0.2,0.2", B="0.2,0.5,0.5", bPickA=Get bIsTargeting())) -> (Output_Get:Vector)
   TryGetPawnOwner()
  ]}
]
variables [acc_scaler_vector {type=Vector}]
